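; Blueprint for a pinsetter prop in the Asteria bowling alley environment set.
; Its construction script calls a ValidateLightVisuals function.
; That function has an empty body in this file, so no light logic is defined here.
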(fn ConstructionScript
  (Default|ValidateLightVisuals))

(fn ValidateLightVisuals)
